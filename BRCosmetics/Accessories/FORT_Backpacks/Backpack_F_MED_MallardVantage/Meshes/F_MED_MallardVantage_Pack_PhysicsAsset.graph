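# Physics asset for the MallardVantage backpack cosmetic mesh, defining collision bodies and joints for its dynamic body and two dangling legs.
# A large root capsule connects to a body capsule, which carries a small sphere master bone per leg, all with free angular motion.
# Each master drives a six-capsule chain whose links use tighter swing and very narrow twist limits, letting the legs bend with little twisting.
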PhysicsAsset /BRCosmetics/Accessories/FORT_Backpacks/Backpack_F_MED_MallardVantage/Meshes/F_MED_MallardVantage_Pack_PhysicsAsset
bodies {
 dyn_body [{shapeType=Capsule, length=11, radius=8, center=<2 2.5 0>, rotation=<rot 0 -15 90>}]
 dyn_l_leg_01 [{shapeType=Capsule, length=1, radius=1, shapeName=dyn_l_leg_01_capsule, center=<1.5 0 0>, rotation=<rot 90 0 0>}]
 dyn_l_leg_02 [{shapeType=Capsule, length=1, radius=1, shapeName=dyn_l_leg_02_capsule, center=<1.5 0 0>, rotation=<rot 90 0 0>}]
 dyn_l_leg_03 [{shapeType=Capsule, length=2, radius=1, shapeName=dyn_l_leg_03_capsule, center=<2 0 0>, rotation=<rot 90 0 0>}]
 dyn_l_leg_04 [{shapeType=Capsule, length=1, radius=1, shapeName=dyn_l_leg_04_capsule, center=<1.5 0 0>, rotation=<rot 90 0 0>}]
 dyn_l_leg_05 [{shapeType=Capsule, length=1, radius=1, shapeName=dyn_l_leg_05_capsule, center=<1.5 0 0>, rotation=<rot 90 0 0>}]
 dyn_l_leg_06 [{shapeType=Capsule, length=1, radius=1, shapeName=dyn_l_leg_06_capsule, center=<1.5 0 0>, rotation=<rot 90 0 0>}]
 dyn_l_leg_mstr [{shapeType=Sphere, radius=0.5, shapeName=dyn_l_leg_mstr_sphere}]
 dyn_r_leg_01 [{shapeType=Capsule, length=1, radius=1, shapeName=dyn_r_leg_01_capsule, center=<1.5 0 0>, rotation=<rot 90 0 0>}]
 dyn_r_leg_02 [{shapeType=Capsule, length=1, radius=1, shapeName=dyn_r_leg_02_capsule, center=<1.5 0 0>, rotation=<rot 90 0 0>}]
 dyn_r_leg_03 [{shapeType=Capsule, length=1, radius=1, shapeName=dyn_r_leg_03_capsule, center=<1.5 0 0>, rotation=<rot 90 0 0>}]
 dyn_r_leg_04 [{shapeType=Capsule, length=1, radius=1, shapeName=dyn_r_leg_04_capsule, center=<1.5 0 0>, rotation=<rot 90 0 0>}]
 dyn_r_leg_05 [{shapeType=Capsule, length=1, radius=1, shapeName=dyn_r_leg_05_capsule, center=<1.5 0 0>, rotation=<rot 90 0 0>}]
 dyn_r_leg_06 [{shapeType=Capsule, length=1, radius=1, shapeName=dyn_r_leg_06_capsule, center=<1.5 0 0>, rotation=<rot 90 0 0>}]
 dyn_r_leg_mstr [{shapeType=Sphere, radius=0.5, shapeName=dyn_r_leg_mstr_sphere}]
 root [{shapeType=Capsule, length=25, radius=11, shapeName=root_capsule, center=<0 10 -10>, rotation=<rot -0.0002 -90 0>}]
}
constraints [
 {
  bone1Name=dyn_body
  bone2Name=root
  swing1LimitDegrees=45
  swing1Motion=Free
  swing2LimitDegrees=45
  swing2Motion=Free
  twistLimitDegrees=45
  twistMotion=Free
 }
 {
  bone1Name=dyn_l_leg_mstr
  bone2Name=dyn_body
  swing1LimitDegrees=45
  swing1Motion=Free
  swing2LimitDegrees=45
  swing2Motion=Free
  twistLimitDegrees=45
  twistMotion=Free
 }
 {bone1Name=dyn_l_leg_01, bone2Name=dyn_l_leg_mstr, swing1LimitDegrees=30, swing2LimitDegrees=20, twistLimitDegrees=5}
 {bone1Name=dyn_l_leg_02, bone2Name=dyn_l_leg_01, swing1LimitDegrees=30, swing2LimitDegrees=20, twistLimitDegrees=5}
 {bone1Name=dyn_l_leg_03, bone2Name=dyn_l_leg_02, swing1LimitDegrees=30, swing2LimitDegrees=20, twistLimitDegrees=5}
 {bone1Name=dyn_l_leg_04, bone2Name=dyn_l_leg_03, swing1LimitDegrees=30, swing2LimitDegrees=20, twistLimitDegrees=5}
 {bone1Name=dyn_l_leg_05, bone2Name=dyn_l_leg_04, swing1LimitDegrees=30, swing2LimitDegrees=20, twistLimitDegrees=5}
 {bone1Name=dyn_l_leg_06, bone2Name=dyn_l_leg_05, swing1LimitDegrees=30, swing2LimitDegrees=20, twistLimitDegrees=5}
 {
  bone1Name=dyn_r_leg_mstr
  bone2Name=dyn_body
  swing1LimitDegrees=45
  swing1Motion=Free
  swing2LimitDegrees=45
  swing2Motion=Free
  twistLimitDegrees=45
  twistMotion=Free
 }
 {bone1Name=dyn_r_leg_01, bone2Name=dyn_r_leg_mstr, swing1LimitDegrees=30, swing2LimitDegrees=20, twistLimitDegrees=5}
 {bone1Name=dyn_r_leg_02, bone2Name=dyn_r_leg_01, swing1LimitDegrees=30, swing2LimitDegrees=20, twistLimitDegrees=5}
 {bone1Name=dyn_r_leg_03, bone2Name=dyn_r_leg_02, swing1LimitDegrees=30, swing2LimitDegrees=20, twistLimitDegrees=5}
 {bone1Name=dyn_r_leg_04, bone2Name=dyn_r_leg_03, swing1LimitDegrees=30, swing2LimitDegrees=20, twistLimitDegrees=5}
 {bone1Name=dyn_r_leg_05, bone2Name=dyn_r_leg_04, swing1LimitDegrees=30, swing2LimitDegrees=20, twistLimitDegrees=5}
 {bone1Name=dyn_r_leg_06, bone2Name=dyn_r_leg_05, swing1LimitDegrees=30, swing2LimitDegrees=20, twistLimitDegrees=5}
]
