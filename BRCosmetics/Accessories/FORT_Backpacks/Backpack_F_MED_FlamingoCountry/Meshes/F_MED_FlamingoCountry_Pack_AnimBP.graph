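# Animation Blueprint for the FlamingoCountry backpack cosmetic, adding secondary motion to its feathers, wings, legs, eyes, pouches and zippers.
# A cached AnimDynamics pose feeds RigidBody-plus-AnimDynamics chains, selected by front-end and DBNO state, or leg Trail controllers, chosen by bRigidBodyEnabled.
# The DBNO chain uses the MallardVantage physics asset.
# Finally, the pack bone shifts and rolls with spine yaw, at reduced alpha while jumping.
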AnimBlueprint /BRCosmetics/Accessories/FORT_Backpacks/Backpack_F_MED_FlamingoCountry/Meshes/F_MED_FlamingoCountry_Pack_AnimBP
target_skeleton=&/F_MED_FlamingoCountry_Pack_Skeleton
graphs [
 AnimGraph {nodes [
   RigidBody!(ComponentPose, Alpha, OverrideWorldGravity:Vector="0,0,-980", ExternalForce) -> (Pose) {
    bClampLinearTranslationLimitToRefPose=true
    bDefaultToSkeletalMeshPhysicsAsset=false
    overridePhysicsAsset=&/F_MED_FlamingoCountry_Pack_PhysicsAsset
    simulationSpace=WorldSpace
    alphaBoolBlend {blendOption=Linear}
    baseBoneRef {boneName=root}
    componentAppliedLinearAccClamp=<7000 7000 7000>
    componentLinearAccScale=<0.4 0.4 0.5>
    simSpaceSettings {
     maxAngularAcceleration=1000
     maxAngularVelocity=3
     maxLinearAcceleration=700
     maxLinearVelocity=500
     velocityScaleZ=5
     externalLinearDragV=<0.2 0.2>
    }
   }
   n1 = RigidBody(ComponentPose=Local To Component(LocalPose=Use cached pose 'Anim Dynamics'()), OverrideWorldGravity="0,0,-980") {
    bClampLinearTranslationLimitToRefPose=true
    bDefaultToSkeletalMeshPhysicsAsset=false
    bOverrideWorldGravity=true
    overridePhysicsAsset=/BRCosmetics/Accessories/FORT_Backpacks/Backpack_F_MED_MallardVantage/Meshes/F_MED_MallardVantage_Pack_PhysicsAsset
    simulationSpace=BaseBoneSpace
    alphaBoolBlend {blendOption=Linear}
    baseBoneRef {boneName=root}
    componentAppliedLinearAccClamp=<3000 3000 3000>
    componentLinearAccScale=<0.3 0.3 0.3>
    componentLinearVelScale=<0.15 0.15 0.15>
    simSpaceSettings {
     maxAngularAcceleration=1000
     maxAngularVelocity=3
     maxLinearAcceleration=700
     maxLinearVelocity=500
     velocityScaleZ=5
     worldAlpha=1
     externalLinearDragV=<0.2 0.2>
    }
   }
   n2 = Anim Dynamics - Bone: dyn_l_wing_01(ComponentPose=@n1, Alpha=1, ExternalForce="0,0,0") {
    angularDampingOverride=0.9999
    angularSpringConstant=10
    bAngularSpring=true
    bGravityOverrideInSimSpace=true
    bOverrideAngularDamping=true
    simulationSpace=RootRelative
    alphaBoolBlend {blendOption=Linear}
    boundBone {boneName=dyn_l_wing_01}
    componentAppliedLinearAccClamp=<5000 5000 5000>
    componentLinearAccScale=<0.5 0.5 0.5>
    componentLinearVelScale=<0.5 0.5 0.5>
    physicsBodyDefinitions [{
      sphereCollisionRadius=0
      boundBone {boneName=dyn_l_wing_01}
      boxExtents=<10 20 20>
      constraintSetup {angularLimitsMax=<1 2 5>, angularLimitsMin=<-1 -2 0>, angularTarget=<0 0 0>}
      localJointOffset=<1 -0.2 0>
     }]
   }
   n3 = Anim Dynamics - Bone: dyn_l_pocket_01(ComponentPose=@n2, Alpha=1, ExternalForce="0,0,0") {
    angularDampingOverride=0.9999
    angularSpringConstant=10
    bAngularSpring=true
    bGravityOverrideInSimSpace=true
    bOverrideAngularDamping=true
    simulationSpace=RootRelative
    alphaBoolBlend {blendOption=Linear}
    boundBone {boneName=dyn_l_pocket_01}
    componentAppliedLinearAccClamp=<5000 5000 5000>
    componentLinearAccScale=<0.5 0.5 0.5>
    componentLinearVelScale=<0.5 0.5 0.5>
    physicsBodyDefinitions [{
      sphereCollisionRadius=0
      boundBone {boneName=dyn_l_pocket_01}
      boxExtents=<10 20 20>
      constraintSetup {angularLimitsMax=<1 5 25>, angularLimitsMin=<-1 -5 0>, angularTarget=<0 0 0>}
      localJointOffset=<1 -0.2 0>
     }]
   }
   n4 = Anim Dynamics - Bone: dyn_l_wing_feather_01(ComponentPose=@n3, Alpha=1, ExternalForce="0,0,0") {
    angularDampingOverride=0.9999
    angularSpringConstant=10
    bAngularSpring=true
    bGravityOverrideInSimSpace=true
    bOverrideAngularDamping=true
    simulationSpace=RootRelative
    alphaBoolBlend {blendOption=Linear}
    boundBone {boneName=dyn_l_wing_feather_01}
    componentAppliedLinearAccClamp=<5000 5000 5000>
    componentLinearAccScale=<0.5 0.5 0.5>
    componentLinearVelScale=<0.5 0.5 0.5>
    physicsBodyDefinitions [{
      sphereCollisionRadius=0
      boundBone {boneName=dyn_l_wing_feather_01}
      boxExtents=<10 20 20>
      constraintSetup {angularLimitsMax=<1 5 25>, angularLimitsMin=<-1 -5 0>, angularTarget=<0 0 0>}
      localJointOffset=<1 -0.2 0>
     }]
   }
   n5 = Anim Dynamics - Bone: dyn_l_wing_feather_02(ComponentPose=@n4, Alpha=1, ExternalForce="0,0,0") {
    angularDampingOverride=0.9999
    angularSpringConstant=10
    bAngularSpring=true
    bGravityOverrideInSimSpace=true
    bOverrideAngularDamping=true
    simulationSpace=RootRelative
    alphaBoolBlend {blendOption=Linear}
    boundBone {boneName=dyn_l_wing_feather_02}
    componentAppliedLinearAccClamp=<5000 5000 5000>
    componentLinearAccScale=<0.5 0.5 0.5>
    componentLinearVelScale=<0.5 0.5 0.5>
    physicsBodyDefinitions [{
      sphereCollisionRadius=0
      boundBone {boneName=dyn_l_wing_feather_02}
      boxExtents=<10 20 20>
      constraintSetup {angularLimitsMax=<1 5 25>, angularLimitsMin=<-1 -5 0>, angularTarget=<0 0 0>}
      localJointOffset=<1 -0.2 0>
     }]
   }
   n6 = Anim Dynamics - Bone: dyn_l_wing_feather_03(ComponentPose=@n5, Alpha=1, ExternalForce="0,0,0") {
    angularDampingOverride=0.9999
    angularSpringConstant=10
    bAngularSpring=true
    bGravityOverrideInSimSpace=true
    bOverrideAngularDamping=true
    simulationSpace=RootRelative
    alphaBoolBlend {blendOption=Linear}
    boundBone {boneName=dyn_l_wing_feather_03}
    componentAppliedLinearAccClamp=<5000 5000 5000>
    componentLinearAccScale=<0.5 0.5 0.5>
    componentLinearVelScale=<0.5 0.5 0.5>
    physicsBodyDefinitions [{
      sphereCollisionRadius=0
      boundBone {boneName=dyn_l_wing_feather_03}
      boxExtents=<10 20 20>
      constraintSetup {angularLimitsMax=<1 5 25>, angularLimitsMin=<-1 -5 0>, angularTarget=<0 0 0>}
      localJointOffset=<1 -0.2 0>
     }]
   }
   n7 = Anim Dynamics - Bone: dyn_r_wing_01(ComponentPose=@n6, Alpha=1, ExternalForce="0,0,0") {
    angularDampingOverride=0.9999
    angularSpringConstant=10
    bAngularSpring=true
    bGravityOverrideInSimSpace=true
    bOverrideAngularDamping=true
    simulationSpace=RootRelative
    alphaBoolBlend {blendOption=Linear}
    boundBone {boneName=dyn_r_wing_01}
    componentAppliedLinearAccClamp=<5000 5000 5000>
    componentLinearAccScale=<0.5 0.5 0.5>
    componentLinearVelScale=<0.5 0.5 0.5>
    physicsBodyDefinitions [{
      sphereCollisionRadius=0
      boundBone {boneName=dyn_r_wing_01}
      boxExtents=<10 20 20>
      constraintSetup {angularLimitsMax=<1 2 5>, angularLimitsMin=<-1 -2 0>, angularTarget=<0 0 0>}
      localJointOffset=<1 -0.2 0>
     }]
   }
   n8 = Anim Dynamics - Bone: dyn_r_pocket_01(ComponentPose=@n7, Alpha=1, ExternalForce="0,0,0") {
    angularDampingOverride=0.9999
    angularSpringConstant=10
    bAngularSpring=true
    bGravityOverrideInSimSpace=true
    bOverrideAngularDamping=true
    simulationSpace=RootRelative
    alphaBoolBlend {blendOption=Linear}
    boundBone {boneName=dyn_r_pocket_01}
    componentAppliedLinearAccClamp=<5000 5000 5000>
    componentLinearAccScale=<0.5 0.5 0.5>
    componentLinearVelScale=<0.5 0.5 0.5>
    physicsBodyDefinitions [{
      sphereCollisionRadius=0
      boundBone {boneName=dyn_r_pocket_01}
      boxExtents=<10 20 20>
      constraintSetup {angularLimitsMax=<1 5 0>, angularLimitsMin=<-1 -5 -25>, angularTarget=<0 0 0>}
      localJointOffset=<1 -0.2 0>
     }]
   }
   n9 = Anim Dynamics - Bone: dyn_r_wing_feather_01(ComponentPose=@n8, Alpha=1, ExternalForce="0,0,0") {
    angularDampingOverride=0.9999
    angularSpringConstant=10
    bAngularSpring=true
    bGravityOverrideInSimSpace=true
    bOverrideAngularDamping=true
    simulationSpace=RootRelative
    alphaBoolBlend {blendOption=Linear}
    boundBone {boneName=dyn_r_wing_feather_01}
    componentAppliedLinearAccClamp=<5000 5000 5000>
    componentLinearAccScale=<0.5 0.5 0.5>
    componentLinearVelScale=<0.5 0.5 0.5>
    physicsBodyDefinitions [{
      sphereCollisionRadius=0
      boundBone {boneName=dyn_r_wing_feather_01}
      boxExtents=<10 20 20>
      constraintSetup {angularLimitsMax=<1 5 0>, angularLimitsMin=<-1 -5 -25>, angularTarget=<0 0 0>}
      localJointOffset=<1 -0.2 0>
     }]
   }
   n10 = Anim Dynamics - Bone: dyn_r_wing_feather_02(ComponentPose=@n9, Alpha=1, ExternalForce="0,0,0") {
    angularDampingOverride=0.9999
    angularSpringConstant=10
    bAngularSpring=true
    bGravityOverrideInSimSpace=true
    bOverrideAngularDamping=true
    simulationSpace=RootRelative
    alphaBoolBlend {blendOption=Linear}
    boundBone {boneName=dyn_r_wing_feather_02}
    componentAppliedLinearAccClamp=<5000 5000 5000>
    componentLinearAccScale=<0.5 0.5 0.5>
    componentLinearVelScale=<0.5 0.5 0.5>
    physicsBodyDefinitions [{
      sphereCollisionRadius=0
      boundBone {boneName=dyn_r_wing_feather_02}
      boxExtents=<10 20 20>
      constraintSetup {angularLimitsMax=<1 5 0>, angularLimitsMin=<-1 -5 -25>, angularTarget=<0 0 0>}
      localJointOffset=<1 -0.2 0>
     }]
   }
   n11 = Anim Dynamics - Bone: dyn_r_wing_feather_03(ComponentPose=@n10, Alpha=1, ExternalForce="0,0,0") {
    angularDampingOverride=0.9999
    angularSpringConstant=10
    bAngularSpring=true
    bGravityOverrideInSimSpace=true
    bOverrideAngularDamping=true
    simulationSpace=RootRelative
    alphaBoolBlend {blendOption=Linear}
    boundBone {boneName=dyn_r_wing_feather_03}
    componentAppliedLinearAccClamp=<5000 5000 5000>
    componentLinearAccScale=<0.5 0.5 0.5>
    componentLinearVelScale=<0.5 0.5 0.5>
    physicsBodyDefinitions [{
      sphereCollisionRadius=0
      boundBone {boneName=dyn_r_wing_feather_03}
      boxExtents=<10 20 20>
      constraintSetup {angularLimitsMax=<1 5 0>, angularLimitsMin=<-1 -5 -25>, angularTarget=<0 0 0>}
      localJointOffset=<1 -0.2 0>
     }]
   }
   n12 = Anim Dynamics - Bone: dyn_l_zipper_01(ComponentPose=@n11, Alpha=1, ExternalForce="0,0,0") {
    angularDampingOverride=0.9999
    bOverrideAngularDamping=true
    bUseGravityOverride=true
    bUsePlanarLimit=false
    simulationSpace=RootRelative
    alphaBoolBlend {blendOption=Linear}
    boundBone {boneName=dyn_l_zipper_01}
    componentAppliedLinearAccClamp=<3000 3000 3000>
    componentLinearAccScale=<0.5 0.5 0.5>
    componentLinearVelScale=<0.25 0.25 0.25>
    gravityOverride {z=-980}
    physicsBodyDefinitions [{
      boundBone {boneName=dyn_l_zipper_01}
      boxExtents=<2 10 10>
      constraintSetup {angularLimitsMax=<2 5 12>, angularLimitsMin=<-2 -30 -30>, angularTarget=<0 0 0>}
      localJointOffset=<3 0 0>
     }]
   }
   n13 = Anim Dynamics - Bone: dyn_r_zipper_01(ComponentPose=@n12, Alpha=1, ExternalForce="0,0,0") {
    angularDampingOverride=0.9999
    bOverrideAngularDamping=true
    bUseGravityOverride=true
    bUsePlanarLimit=false
    simulationSpace=RootRelative
    alphaBoolBlend {blendOption=Linear}
    boundBone {boneName=dyn_r_zipper_01}
    componentAppliedLinearAccClamp=<3000 3000 3000>
    componentLinearAccScale=<0.5 0.5 0.5>
    componentLinearVelScale=<0.25 0.25 0.25>
    gravityOverride {z=-980}
    physicsBodyDefinitions [{
      boundBone {boneName=dyn_r_zipper_01}
      boxExtents=<2 10 10>
      constraintSetup {angularLimitsMax=<2 5 15>, angularLimitsMin=<-2 -30 -30>, angularTarget=<0 0 0>}
      localJointOffset=<3 0 0>
     }]
   }
   n14 = RigidBody(ComponentPose=Local To Component(LocalPose=Use cached pose 'Anim Dynamics'()), OverrideWorldGravity="0,0,-980") {
    bClampLinearTranslationLimitToRefPose=true
    bDefaultToSkeletalMeshPhysicsAsset=false
    bOverrideWorldGravity=true
    overridePhysicsAsset=&F_MED_FlamingoCountry_Pack_PhysicsAsset
    simulationSpace=BaseBoneSpace
    alphaBoolBlend {blendOption=Linear}
    baseBoneRef {boneName=root}
    componentAppliedLinearAccClamp=<3000 3000 3000>
    componentLinearAccScale=<0.3 0.3 0.3>
    componentLinearVelScale=<0.15 0.15 0.15>
    simSpaceSettings {
     maxAngularAcceleration=1000
     maxAngularVelocity=3
     maxLinearAcceleration=700
     maxLinearVelocity=500
     velocityScaleZ=5
     worldAlpha=1
     externalLinearDragV=<0.2 0.2>
    }
   }
   n15 = Component To Local(ComponentPose=@n14)
   n16 = RigidBody(ComponentPose=Local To Component(LocalPose=Use cached pose 'Anim Dynamics'()), OverrideWorldGravity="0,0,-980") {
    bClampLinearTranslationLimitToRefPose=true
    bDefaultToSkeletalMeshPhysicsAsset=false
    bOverrideWorldGravity=true
    overridePhysicsAsset=&F_MED_FlamingoCountry_Pack_PhysicsAsset
    simulationSpace=WorldSpace
    alphaBoolBlend {blendOption=Linear}
    baseBoneRef {boneName=root}
    componentAppliedLinearAccClamp=<1500 1500 1500>
    componentLinearAccScale=<0.3 0.3 0.3>
    componentLinearVelScale=<0.15 0.15 0.15>
    simSpaceSettings {
     maxAngularAcceleration=1000
     maxAngularVelocity=3
     maxLinearAcceleration=700
     maxLinearVelocity=500
     velocityScaleZ=5
     externalLinearDragV=<0.2 0.2>
    }
   }
   n17 = Component To Local(ComponentPose=@n16)
   n18 = Get bIsInFrontEnd()
   n19 = Anim Dynamics - Bone: dyn_l_wing_01(
    ComponentPose=Local To Component(LocalPose=Blend Poses by bool(BlendPose_0=@n17, BlendPose_1=@n15, bActiveValue=@n18) {blendType=Linear})
    Alpha=1
    ExternalForce="0,0,0"
   ) {
    angularDampingOverride=0.9999
    angularSpringConstant=10
    bAngularSpring=true
    bGravityOverrideInSimSpace=true
    bOverrideAngularDamping=true
    simulationSpace=RootRelative
    alphaBoolBlend {blendOption=Linear}
    boundBone {boneName=dyn_l_wing_01}
    componentAppliedLinearAccClamp=<5000 5000 5000>
    componentLinearAccScale=<0.5 0.5 0.5>
    componentLinearVelScale=<0.5 0.5 0.5>
    physicsBodyDefinitions [{
      sphereCollisionRadius=0
      boundBone {boneName=dyn_l_wing_01}
      boxExtents=<10 20 20>
      constraintSetup {angularLimitsMax=<1 10 7>, angularLimitsMin=<-1 -10 -5>, angularTarget=<0 0 0>}
      localJointOffset=<1 -0.2 0>
     }]
   }
   n20 = Anim Dynamics - Bone: dyn_l_pocket_01(ComponentPose=@n19, Alpha=1, ExternalForce="0,0,0") {
    angularDampingOverride=0.9999
    angularSpringConstant=10
    bAngularSpring=true
    bGravityOverrideInSimSpace=true
    bOverrideAngularDamping=true
    simulationSpace=RootRelative
    alphaBoolBlend {blendOption=Linear}
    boundBone {boneName=dyn_l_pocket_01}
    componentAppliedLinearAccClamp=<5000 5000 5000>
    componentLinearAccScale=<0.5 0.5 0.5>
    componentLinearVelScale=<0.5 0.5 0.5>
    physicsBodyDefinitions [{
      sphereCollisionRadius=0
      boundBone {boneName=dyn_l_pocket_01}
      boxExtents=<10 20 20>
      constraintSetup {angularLimitsMax=<1 5 25>, angularLimitsMin=<-1 -5 0>, angularTarget=<0 0 0>}
      localJointOffset=<1 -0.2 0>
     }]
   }
   n21 = Anim Dynamics - Bone: dyn_l_wing_feather_01(ComponentPose=@n20, Alpha=1, ExternalForce="0,0,0") {
    angularDampingOverride=0.9999
    angularSpringConstant=10
    bAngularSpring=true
    bGravityOverrideInSimSpace=true
    bOverrideAngularDamping=true
    simulationSpace=RootRelative
    alphaBoolBlend {blendOption=Linear}
    boundBone {boneName=dyn_l_wing_feather_01}
    componentAppliedLinearAccClamp=<5000 5000 5000>
    componentLinearAccScale=<0.5 0.5 0.5>
    componentLinearVelScale=<0.5 0.5 0.5>
    physicsBodyDefinitions [{
      sphereCollisionRadius=0
      boundBone {boneName=dyn_l_wing_feather_01}
      boxExtents=<10 20 20>
      constraintSetup {angularLimitsMax=<1 5 25>, angularLimitsMin=<-1 -5 0>, angularTarget=<0 0 0>}
      localJointOffset=<1 -0.2 0>
     }]
   }
   n22 = Anim Dynamics - Bone: dyn_l_wing_feather_02(ComponentPose=@n21, Alpha=1, ExternalForce="0,0,0") {
    angularDampingOverride=0.9999
    angularSpringConstant=10
    bAngularSpring=true
    bGravityOverrideInSimSpace=true
    bOverrideAngularDamping=true
    simulationSpace=RootRelative
    alphaBoolBlend {blendOption=Linear}
    boundBone {boneName=dyn_l_wing_feather_02}
    componentAppliedLinearAccClamp=<5000 5000 5000>
    componentLinearAccScale=<0.5 0.5 0.5>
    componentLinearVelScale=<0.5 0.5 0.5>
    physicsBodyDefinitions [{
      sphereCollisionRadius=0
      boundBone {boneName=dyn_l_wing_feather_02}
      boxExtents=<10 20 20>
      constraintSetup {angularLimitsMax=<1 5 25>, angularLimitsMin=<-1 -5 0>, angularTarget=<0 0 0>}
      localJointOffset=<1 -0.2 0>
     }]
   }
   n23 = Anim Dynamics - Bone: dyn_l_wing_feather_03(ComponentPose=@n22, Alpha=1, ExternalForce="0,0,0") {
    angularDampingOverride=0.9999
    angularSpringConstant=10
    bAngularSpring=true
    bGravityOverrideInSimSpace=true
    bOverrideAngularDamping=true
    simulationSpace=RootRelative
    alphaBoolBlend {blendOption=Linear}
    boundBone {boneName=dyn_l_wing_feather_03}
    componentAppliedLinearAccClamp=<5000 5000 5000>
    componentLinearAccScale=<0.5 0.5 0.5>
    componentLinearVelScale=<0.5 0.5 0.5>
    physicsBodyDefinitions [{
      sphereCollisionRadius=0
      boundBone {boneName=dyn_l_wing_feather_03}
      boxExtents=<10 20 20>
      constraintSetup {angularLimitsMax=<1 5 25>, angularLimitsMin=<-1 -5 0>, angularTarget=<0 0 0>}
      localJointOffset=<1 -0.2 0>
     }]
   }
   n24 = Anim Dynamics - Bone: dyn_r_wing_01(ComponentPose=@n23, Alpha=1, ExternalForce="0,0,0") {
    angularDampingOverride=0.9999
    angularSpringConstant=10
    bAngularSpring=true
    bGravityOverrideInSimSpace=true
    bOverrideAngularDamping=true
    simulationSpace=RootRelative
    alphaBoolBlend {blendOption=Linear}
    boundBone {boneName=dyn_r_wing_01}
    componentAppliedLinearAccClamp=<5000 5000 5000>
    componentLinearAccScale=<0.5 0.5 0.5>
    componentLinearVelScale=<0.5 0.5 0.5>
    physicsBodyDefinitions [{
      sphereCollisionRadius=0
      boundBone {boneName=dyn_r_wing_01}
      boxExtents=<10 20 20>
      constraintSetup {angularLimitsMax=<1 10 7>, angularLimitsMin=<-1 -10 -5>, angularTarget=<0 0 0>}
      localJointOffset=<1 -0.2 0>
     }]
   }
   n25 = Anim Dynamics - Bone: dyn_r_pocket_01(ComponentPose=@n24, Alpha=1, ExternalForce="0,0,0") {
    angularDampingOverride=0.9999
    angularSpringConstant=10
    bAngularSpring=true
    bGravityOverrideInSimSpace=true
    bOverrideAngularDamping=true
    simulationSpace=RootRelative
    alphaBoolBlend {blendOption=Linear}
    boundBone {boneName=dyn_r_pocket_01}
    componentAppliedLinearAccClamp=<5000 5000 5000>
    componentLinearAccScale=<0.5 0.5 0.5>
    componentLinearVelScale=<0.5 0.5 0.5>
    physicsBodyDefinitions [{
      sphereCollisionRadius=0
      boundBone {boneName=dyn_r_pocket_01}
      boxExtents=<10 20 20>
      constraintSetup {angularLimitsMax=<1 5 0>, angularLimitsMin=<-1 -5 -25>, angularTarget=<0 0 0>}
      localJointOffset=<1 -0.2 0>
     }]
   }
   n26 = Anim Dynamics - Bone: dyn_r_wing_feather_01(ComponentPose=@n25, Alpha=1, ExternalForce="0,0,0") {
    angularDampingOverride=0.9999
    angularSpringConstant=10
    bAngularSpring=true
    bGravityOverrideInSimSpace=true
    bOverrideAngularDamping=true
    simulationSpace=RootRelative
    alphaBoolBlend {blendOption=Linear}
    boundBone {boneName=dyn_r_wing_feather_01}
    componentAppliedLinearAccClamp=<5000 5000 5000>
    componentLinearAccScale=<0.5 0.5 0.5>
    componentLinearVelScale=<0.5 0.5 0.5>
    physicsBodyDefinitions [{
      sphereCollisionRadius=0
      boundBone {boneName=dyn_r_wing_feather_01}
      boxExtents=<10 20 20>
      constraintSetup {angularLimitsMax=<1 5 0>, angularLimitsMin=<-1 -5 -25>, angularTarget=<0 0 0>}
      localJointOffset=<1 -0.2 0>
     }]
   }
   n27 = Anim Dynamics - Bone: dyn_r_wing_feather_02(ComponentPose=@n26, Alpha=1, ExternalForce="0,0,0") {
    angularDampingOverride=0.9999
    angularSpringConstant=10
    bAngularSpring=true
    bGravityOverrideInSimSpace=true
    bOverrideAngularDamping=true
    simulationSpace=RootRelative
    alphaBoolBlend {blendOption=Linear}
    boundBone {boneName=dyn_r_wing_feather_02}
    componentAppliedLinearAccClamp=<5000 5000 5000>
    componentLinearAccScale=<0.5 0.5 0.5>
    componentLinearVelScale=<0.5 0.5 0.5>
    physicsBodyDefinitions [{
      sphereCollisionRadius=0
      boundBone {boneName=dyn_r_wing_feather_02}
      boxExtents=<10 20 20>
      constraintSetup {angularLimitsMax=<1 5 0>, angularLimitsMin=<-1 -5 -25>, angularTarget=<0 0 0>}
      localJointOffset=<1 -0.2 0>
     }]
   }
   n28 = Anim Dynamics - Bone: dyn_r_wing_feather_03(ComponentPose=@n27, Alpha=1, ExternalForce="0,0,0") {
    angularDampingOverride=0.9999
    angularSpringConstant=10
    bAngularSpring=true
    bGravityOverrideInSimSpace=true
    bOverrideAngularDamping=true
    simulationSpace=RootRelative
    alphaBoolBlend {blendOption=Linear}
    boundBone {boneName=dyn_r_wing_feather_03}
    componentAppliedLinearAccClamp=<5000 5000 5000>
    componentLinearAccScale=<0.5 0.5 0.5>
    componentLinearVelScale=<0.5 0.5 0.5>
    physicsBodyDefinitions [{
      sphereCollisionRadius=0
      boundBone {boneName=dyn_r_wing_feather_03}
      boxExtents=<10 20 20>
      constraintSetup {angularLimitsMax=<1 5 0>, angularLimitsMin=<-1 -5 -25>, angularTarget=<0 0 0>}
      localJointOffset=<1 -0.2 0>
     }]
   }
   n29 = Anim Dynamics - Bone: dyn_l_zipper_01(ComponentPose=@n28, Alpha=1, ExternalForce="0,0,0") {
    angularDampingOverride=0.9999
    bOverrideAngularDamping=true
    bUseGravityOverride=true
    bUsePlanarLimit=false
    simulationSpace=RootRelative
    alphaBoolBlend {blendOption=Linear}
    boundBone {boneName=dyn_l_zipper_01}
    componentAppliedLinearAccClamp=<3000 3000 3000>
    componentLinearAccScale=<0.5 0.5 0.5>
    componentLinearVelScale=<0.25 0.25 0.25>
    gravityOverride {z=-980}
    physicsBodyDefinitions [{
      boundBone {boneName=dyn_l_zipper_01}
      boxExtents=<2 10 10>
      constraintSetup {angularLimitsMax=<2 5 10>, angularLimitsMin=<-2 -30 -30>, angularTarget=<0 0 0>}
      localJointOffset=<3 0 0>
     }]
   }
   n30 = Anim Dynamics - Bone: dyn_r_zipper_01(ComponentPose=@n29, Alpha=1, ExternalForce="0,0,0") {
    angularDampingOverride=0.9999
    bOverrideAngularDamping=true
    bUseGravityOverride=true
    bUsePlanarLimit=false
    simulationSpace=RootRelative
    alphaBoolBlend {blendOption=Linear}
    boundBone {boneName=dyn_r_zipper_01}
    componentAppliedLinearAccClamp=<3000 3000 3000>
    componentLinearAccScale=<0.5 0.5 0.5>
    componentLinearVelScale=<0.25 0.25 0.25>
    gravityOverride {z=-980}
    physicsBodyDefinitions [{
      boundBone {boneName=dyn_r_zipper_01}
      boxExtents=<2 10 10>
      constraintSetup {angularLimitsMax=<2 5 10>, angularLimitsMin=<-2 -30 -30>, angularTarget=<0 0 0>}
      localJointOffset=<3 0 0>
     }]
   }
   n31 = Blend Poses by bool(BlendPose_0=Component To Local(ComponentPose=@n13), BlendPose_1=Component To Local(ComponentPose=@n30), bActiveValue=Get bIsDBNO()) {blendType=Linear}
   n32 = Trail controller - Bone: dyn_l_leg_06(ComponentPose=Local To Component(LocalPose=Use cached pose 'Anim Dynamics'()), Alpha=0.8) {
    bLimitRotation=true
    bUsePlanarLimit=true
    chainLength=5
    debugLifeTime=0.1
    maxDeltaTime=0.333
    alphaScaleBiasClamp {bInterpResult=true}
    baseJoint {boneName=root}
    fakeVelocity {x=1}
    planarLimits [{drivingBone {boneName=root}, planeTransform {location=<0 -3 0>, rotation=<rot 0 0 -90>, scale=<1 1 1>}}]
    rotationLimits [{limitMax=<20 180 180>}, {limitMax=<20 180 180>}, {limitMax=<20 180 180>}, {limitMax=<20 180 180>}, {limitMax=<20 180 180>}]
    rotationOffsets [<0 0 0>, <0 0 0>, <0 0 0>, <0 0 0>, <0 0 0>]
    trailBone {boneName=dyn_l_leg_06}
    trailRelaxationSpeed {editorCurveData {keys [<at 0 23 RCIM_Linear>,<at 0.25 20 RCIM_Linear>,<at 0.75 12 RCIM_Linear>,<at 1 10 RCIM_Linear>]}}
   }
   n33 = Trail controller - Bone: dyn_r_leg_06(ComponentPose=@n32, Alpha=0.75) {
    bLimitRotation=true
    bUsePlanarLimit=true
    chainLength=5
    debugLifeTime=0.1
    maxDeltaTime=0.333
    alphaScaleBiasClamp {bInterpResult=true}
    baseJoint {boneName=root}
    fakeVelocity {x=1}
    planarLimits [{drivingBone {boneName=root}, planeTransform {location=<0 -3 0>, rotation=<rot 0 0 -90>, scale=<1 1 1>}}]
    rotationLimits [{limitMax=<20 180 180>}, {limitMax=<20 180 180>}, {limitMax=<20 180 180>}, {limitMax=<20 180 180>}, {limitMax=<20 180 180>}]
    rotationOffsets [<0 0 0>, <0 0 0>, <0 0 0>, <0 0 0>, <0 0 0>]
    trailBone {boneName=dyn_r_leg_06}
    trailRelaxationSpeed {editorCurveData {keys [<at 0 23 RCIM_Linear>,<at 0.25 20 RCIM_Linear>,<at 0.75 12 RCIM_Linear>,<at 1 10 RCIM_Linear>]}}
   }
   n34 = Component To Local(ComponentPose=@n33)
   n35 = Get bRigidBodyEnabled()
   n36 = "Transform (Modify) Bone - Bone: dyn_l_eye_01"(
    ComponentPose=Local To Component(LocalPose=Blend Poses by bool(BlendPose_0=@n31, BlendPose_1=@n34, bActiveValue=@n35) {blendType=Linear})
    Alpha=1
    Translation="0,0,0"
    Rotation="0,0,0"
    Scale="0.98,0.98,0.98"
   ) {
    rotationMode=BMM_Additive
    rotationSpace=BCS_ParentBoneSpace
    scaleMode=BMM_Additive
    scaleSpace=BCS_ParentBoneSpace
    translationMode=BMM_Additive
    translationSpace=BCS_ParentBoneSpace
    alphaBoolBlend {blendOption=Linear}
    boneToModify {boneName=dyn_l_eye_01}
    scale=<0.98 0.98 0.98>
   }
   n37 = "Transform (Modify) Bone - Bone: dyn_r_eye_01"(ComponentPose=@n36, Alpha=1, Translation="0,0,0", Rotation="0,0,0", Scale="0.98,0.98,0.98") {
    scaleMode=BMM_Replace
    scaleSpace=BCS_BoneSpace
    alphaBoolBlend {blendOption=Linear}
    boneToModify {boneName=dyn_r_eye_01}
    scale=<0.98 0.98 0.98>
   }
   n38 = "Transform (Modify) Bone - Bone: pack"(
    ComponentPose=@n37
    Alpha=Get Pack Alpha() -> (Pack Alpha:double)
    Translation=Get Backpack Translate() -> (Backpack Translate:Vector)
    Rotation=Get Backpack Rotate() -> (Backpack Rotate:Rotator)
   ) {
    rotationMode=BMM_Additive
    rotationSpace=BCS_BoneSpace
    translationMode=BMM_Additive
    translationSpace=BCS_BoneSpace
    alphaBoolBlend {blendOption=Linear}
    alphaScaleBiasClamp {bInterpResult=true}
    boneToModify {boneName=pack}
   }
   Output Pose(Result=Component To Local(ComponentPose=@n38)) {layerGroup=None}
   n39 = Anim Dynamics - Bone: dyn_head_feather_01(ComponentPose=Local To Component(LocalPose=Input Pose()), Alpha=1, ExternalForce="0,0,0") {
    angularDampingOverride=0.9999
    angularSpringConstant=10
    bAngularSpring=true
    bGravityOverrideInSimSpace=true
    bOverrideAngularDamping=true
    simulationSpace=RootRelative
    alphaBoolBlend {blendOption=Linear}
    boundBone {boneName=dyn_head_feather_01}
    componentAppliedLinearAccClamp=<5000 5000 5000>
    componentLinearAccScale=<1 1 1>
    componentLinearVelScale=<1 1 1>
    physicsBodyDefinitions [{
      sphereCollisionRadius=0
      boundBone {boneName=dyn_head_feather_01}
      boxExtents=<8 16 16>
      constraintSetup {angularLimitsMax=<1 15 15>, angularLimitsMin=<-1 -15 -15>, angularTarget=<0 0 0>}
      localJointOffset=<1 0 -0.2>
     }]
   }
   n40 = Anim Dynamics - Bone: dyn_head_feather_02(ComponentPose=@n39, Alpha=1, ExternalForce="0,0,0") {
    angularDampingOverride=0.9999
    angularSpringConstant=10
    bAngularSpring=true
    bGravityOverrideInSimSpace=true
    bOverrideAngularDamping=true
    simulationSpace=RootRelative
    alphaBoolBlend {blendOption=Linear}
    boundBone {boneName=dyn_head_feather_02}
    componentAppliedLinearAccClamp=<5000 5000 5000>
    componentLinearAccScale=<0.75 0.75 0.75>
    componentLinearVelScale=<0.75 0.75 0.75>
    physicsBodyDefinitions [{
      sphereCollisionRadius=0
      boundBone {boneName=dyn_head_feather_02}
      boxExtents=<10 20 20>
      constraintSetup {angularLimitsMax=<1 15 15>, angularLimitsMin=<-1 -15 -15>, angularTarget=<0 0 0>}
      localJointOffset=<1 0 -0.2>
     }]
   }
   n41 = "Transform (Modify) Bone - Bone: dyn_l_eye_01"(ComponentPose=@n40, Alpha=1, Translation="0,0,0", Rotation="-20,0,0", Scale="1,1,1") {
    rotationMode=BMM_Replace
    rotationSpace=BCS_ParentBoneSpace
    translationMode=BMM_Replace
    translationSpace=BCS_ParentBoneSpace
    alphaBoolBlend {blendOption=Linear}
    boneToModify {boneName=dyn_l_eye_01}
    rotation {pitch=-20}
   }
   n42 = "Transform (Modify) Bone - Bone: dyn_r_eye_01"(ComponentPose=@n41, Alpha=1, Translation="0,0,0", Rotation="20,0,0", Scale="1,1,1") {
    rotationMode=BMM_Replace
    rotationSpace=BCS_ParentBoneSpace
    translationMode=BMM_Replace
    translationSpace=BCS_ParentBoneSpace
    alphaBoolBlend {blendOption=Linear}
    boneToModify {boneName=dyn_r_eye_01}
    rotation {pitch=20}
   }
   n43 = Anim Dynamics - Bone: dyn_l_eye_01(ComponentPose=@n42, Alpha=1, ExternalForce="0,0,0") {
    angularDampingOverride=0.9999
    angularSpringConstant=10
    bAngularSpring=true
    bGravityOverrideInSimSpace=true
    bOverrideAngularDamping=true
    simulationSpace=RootRelative
    alphaBoolBlend {blendOption=Linear}
    boundBone {boneName=dyn_l_eye_01}
    componentAppliedLinearAccClamp=<10000 10000 10000>
    componentLinearAccScale=<1 1 1>
    componentLinearVelScale=<1 1 1>
    physicsBodyDefinitions [{
      sphereCollisionRadius=0
      boundBone {boneName=dyn_l_eye_01}
      boxExtents=<10 20 20>
      constraintSetup {angularLimitsMax=<1 10 1>, angularLimitsMin=<-1 -10 -1>, angularTarget=<0 0 0>}
      localJointOffset=<2 0 -0.2>
     }]
   }
   n44 = Anim Dynamics - Bone: dyn_r_eye_01(ComponentPose=@n43, Alpha=1, ExternalForce="0,0,0") {
    angularDampingOverride=0.9999
    angularSpringConstant=10
    bAngularSpring=true
    bGravityOverrideInSimSpace=true
    bOverrideAngularDamping=true
    simulationSpace=RootRelative
    alphaBoolBlend {blendOption=Linear}
    boundBone {boneName=dyn_r_eye_01}
    componentAppliedLinearAccClamp=<7000 7000 7000>
    componentLinearAccScale=<0.75 0.75 0.75>
    componentLinearVelScale=<0.75 0.75 0.75>
    physicsBodyDefinitions [{
      sphereCollisionRadius=0
      boundBone {boneName=dyn_r_eye_01}
      boxExtents=<10 20 20>
      constraintSetup {angularLimitsMax=<1 10 1>, angularLimitsMin=<-1 -10 -1>, angularTarget=<0 0 0>}
      localJointOffset=<2 0 -0.2>
     }]
   }
   n45 = Anim Dynamics - Bone: dyn_front_pouch_01(ComponentPose=@n44, Alpha=1, ExternalForce="0,0,0") {
    angularDampingOverride=0.9999
    angularSpringConstant=10
    bAngularSpring=true
    bGravityOverrideInSimSpace=true
    bOverrideAngularDamping=true
    simulationSpace=RootRelative
    alphaBoolBlend {blendOption=Linear}
    boundBone {boneName=dyn_front_pouch_01}
    componentAppliedLinearAccClamp=<5000 5000 5000>
    componentLinearAccScale=<0.5 0.5 0.5>
    componentLinearVelScale=<0.5 0.5 0.5>
    physicsBodyDefinitions [{
      sphereCollisionRadius=0
      boundBone {boneName=dyn_front_pouch_01}
      boxExtents=<10 20 20>
      constraintSetup {angularLimitsMax=<1 5 5>, angularLimitsMin=<-1 -5 -5>, angularTarget=<0 0 0>}
      localJointOffset=<1 0 -0.2>
     }]
   }
   Save cached pose 'Anim Dynamics'(Pose=Component To Local(ComponentPose=@n45)) {cacheName=Anim Dynamics}
  ]}
 EventGraph {nodes [
   Event BlueprintUpdateAnimation()
   > n1 = Sequence()
   > n2 = Gravity Override State Function DEPRECATED(CharacterPartAnimInstance=Self-Reference(), GameTimeInSeconds=GetGameTimeInSeconds(), GravityOverrideParams=Get Gravity Override Params())
   > Set Gravity Override(Gravity Override=@n2)
   > Set Front End Alpha(Front End Alpha:double=SelectFloat(A=0.5, B=1, bPickA=Get bIsInFrontEnd())) -> (Output_Get:double)
   n3 = GetOwnerSkeletalMeshComponent()
   n4 = BreakRotator(InRot=BreakTransform(InTransform=GetDeltaTransformFromRefPose(self=@n3, BoneName=spine_05, BaseName=spine_01)).Rotation)
   Set Backpack Rotate(execute=@n1.then_1, Backpack Rotate:Rotator=MakeRotator(Roll=MapRangeClamped(Value=@n4.Yaw, InRangeA=0, InRangeB=-30, OutRangeB=25))) -> (Output_Get:Rotator)
   > Set Backpack Translate(Backpack Translate:Vector=MakeVector(Y=MapRangeClamped(Value=@n4.Yaw, InRangeA=0, InRangeB=-30, OutRangeA=1, OutRangeB=-7))) -> (Output_Get:Vector)
   > n5 = Branch(Condition=Get bIsJumping())
   > Set Pack Alpha(Pack Alpha:double=0.15) -> (Output_Get:double)
   Set Pack Alpha(execute=@n5.else, Pack Alpha=1)
   TryGetPawnOwner()
  ]}
]
variables [
 Backpack Rotate {type=Rotator}
 Backpack Translate {type=Vector}
 Front End Alpha {type=number}
 Gravity Override {type=Vector}
 Gravity Override Params {type=GravityOverrideParamsStruct}
 Non RBAN Jumping {type=number, default=0.8}
 Non RBAN Not Jumpiing {type=number, default=0.6}
 Pack Alpha {type=number, default=0.15}
]
